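# Physics asset for the M_MED_Halloween_Alt backpack cosmetic, defining collision bodies and joints for its dangling strap and zipper.
# A single box covers the pack, small spheres sit on the two parent bones, and the chain bones use spheres for the strap and boxes for the zipper.
# Two separate two-link constraint chains hang from those parents with limited swing and tight twist; no constraint joins them to the pack body.
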PhysicsAsset /BRCosmetics/Accessories/FORT_Backpacks/M_MED_Halloween_Alt/Meshes/M_MED_Halloween_Alt_Pack_Backpack_Physics
bodies {
 dyn_parent_strap [{shapeType=Sphere, radius=0.505}]
 dyn_parent_zipper [{shapeType=Sphere, radius=0.505}]
 dyn_strap_01 [{shapeType=Sphere, radius=1, center=<1 0 0>}]
 dyn_strap_02 [{shapeType=Sphere, radius=1, center=<1 -0.1 0>}]
 dyn_zipper_01 [{shapeType=Box, extentX=1.7303, extentY=1.3183, extentZ=1.4748, center=<1.1427 -0.0081 -0.0517>, rotation=<rot -0.8367 -90 -0.1881>}]
 dyn_zipper_02 [{shapeType=Box, extentX=5, extentY=1.5, extentZ=9, center=<2.1306 0.1477 -0.015>, rotation=<rot -2.8471 4.1332 -10.8044>}]
 pack [{shapeType=Box, extentX=25, extentY=11.5, extentZ=31.5, center=<-0.0001 -5.9427 -9.716>}]
}
constraints [
 {bone1Name=dyn_zipper_01, bone2Name=dyn_parent_zipper, swing1LimitDegrees=45, swing2LimitDegrees=19.2857, twistLimitDegrees=2}
 {bone1Name=dyn_zipper_02, bone2Name=dyn_zipper_01, swing1LimitDegrees=45, swing2LimitDegrees=21, twistLimitDegrees=2}
 {bone1Name=dyn_strap_01, bone2Name=dyn_parent_strap, swing1LimitDegrees=22.5, swing2LimitDegrees=22.5, twistLimitDegrees=10}
 {bone1Name=dyn_strap_02, bone2Name=dyn_strap_01, swing1LimitDegrees=45, swing2LimitDegrees=45, twistLimitDegrees=3}
]
